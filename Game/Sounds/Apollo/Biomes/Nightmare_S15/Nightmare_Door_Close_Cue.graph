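# Sound cue for a door closing in the Season 15 Nightmare biome.
# It plays a single wave, Nightmare_Door_Close_01, with no randomization or modulation.
# It routes through the Fort_Doors sound class and uses a no-occlusion door attenuation asset.
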SoundCue /Game/Sounds/Apollo/Biomes/Nightmare_S15/Nightmare_Door_Close_Cue
root=WavePlayer_0
nodes [{id=WavePlayer_0, properties {soundWaveAssetPtr=&/Nightmare_Door_Close_01}}]
properties {
 attenuationSettings=/Game/Sounds/Attenuation/Fort_Doors_NoOcclusion_Att
 maxDistance=2400
 soundClassObject=&Fort_Doors
}
