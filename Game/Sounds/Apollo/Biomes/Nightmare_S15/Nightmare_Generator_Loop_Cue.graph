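# Looping ambient sound cue for a generator in the Apollo Nightmare S15 biome, assigned to the outdoor ambient sound class.
# A single looping wave player feeds the output directly at slightly reduced volume.
# Inline attenuation overrides apply logarithmic distance falloff, distance-based low-pass filtering, listener focus that lowers off-focus volume, and trace-based occlusion that muffles and quiets the sound.
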SoundCue /Game/Sounds/Apollo/Biomes/Nightmare_S15/Nightmare_Generator_Loop_Cue
root=WavePlayer_0
nodes [{id=WavePlayer_0, properties {bLooping=true, soundWaveAssetPtr=&/Nightmare_Generator_Loop_01}}]
properties {
 bOverrideAttenuation=true
 maxDistance=4426.1323
 soundClassObject=&Fort_Ambient_Outdoors_Athena
 volumeMultiplier=0.802
 attenuationOverrides {
  bAttenuateWithLPF=true
  bEnableFocusInterpolation=true
  bEnableListenerFocus=true
  bEnableOcclusion=true
  distanceAlgorithm=Logarithmic
  falloffDistance=4391.1133
  lPFFrequencyAtMin=500
  lPFRadiusMax=1000
  lPFRadiusMin=0
  nonFocusVolumeAttenuation=0.5
  occlusionInterpolationTime=1
  occlusionLowPassFilterFrequency=200
  occlusionTraceChannel=ECC_GameTraceChannel6
  occlusionVolumeAttenuation=0.5
  attenuationShapeExtents {x=35.019}
 }
}
