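# Sound cue for one-shot startled-bird ambience in the Nightmare Season 15 jungle biome on the Apollo map.
# A single Random node picks among six bird-startle wave variations with equal weights.
# Playback uses the shared tile ambient one-shot attenuation, routes to the outdoor Athena ambient sound class, and runs at reduced volume.
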SoundCue /Game/Sounds/Apollo/Biomes/Nightmare_S15/Nightmare_Jungle_OneShot_BirdStartle_Cue
root=Random_0
nodes [
 {id=Random_0, children [WavePlayer_1,WavePlayer_2,WavePlayer_3,WavePlayer_4,WavePlayer_5,WavePlayer_6], properties {weights [1,1,1,1,1,1]}}
 {id=WavePlayer_1, properties {soundWaveAssetPtr=&/Nightmare_Jungle_OneShot_BirdStartle_01}}
 {id=WavePlayer_2, properties {soundWaveAssetPtr=&/Nightmare_Jungle_OneShot_BirdStartle_02}}
 {id=WavePlayer_3, properties {soundWaveAssetPtr=&/Nightmare_Jungle_OneShot_BirdStartle_03}}
 {id=WavePlayer_4, properties {soundWaveAssetPtr=&/Nightmare_Jungle_Oneshot_BirdStartle_04}}
 {id=WavePlayer_5, properties {soundWaveAssetPtr=&/Nightmare_Jungle_OneShot_BirdStartle_05}}
 {id=WavePlayer_6, properties {soundWaveAssetPtr=&/Nightmare_Jungle_OneShot_BirdStartle_06}}
]
properties {
 attenuationSettings=/Game/Sounds/Attenuation/Fort_TileAmbient_OneShot_Att
 maxDistance=8000
 soundClassObject=&Fort_Ambient_Outdoors_Athena
 volumeMultiplier=0.501
}
